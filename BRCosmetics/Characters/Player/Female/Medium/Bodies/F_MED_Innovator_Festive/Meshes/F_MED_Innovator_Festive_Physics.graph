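# Physics asset for the F_MED_Innovator_Festive female player body, providing secondary motion for the outfit's hood and two strings.
# Each string chains capsules from a sphere base through parent, 01 and 02 bones, with swing limits loosening toward the tip and tight twist limits.
# The hood is a single capsule constrained to its sphere base.
# Two boxes on spine_05 form the upper-torso shape, with no constraints linking them to the chains.
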PhysicsAsset /BRCosmetics/Characters/Player/Female/Medium/Bodies/F_MED_Innovator_Festive/Meshes/F_MED_Innovator_Festive_Physics
bodies {
 dyn_hood_base [{shapeType=Sphere, radius=0.5, shapeName=dyn_hood_base_sphere}]
 dyn_hood_parent [{shapeType=Capsule, length=4.5916, radius=4.2876, shapeName=dyn_hood_parent_capsule, center=<4.035 0 0.0009>, rotation=<rot 0 0 89.9999>}]
 dyn_string_01_l [{
   shapeType=Capsule
   length=4.1943
   radius=0.5657
   shapeName=dyn_string_01_l_capsule
   center=<2.4212 -0.0147 -0.0425>
   rotation=<rot 0 -90.2507 -89.2869>
  }]
 dyn_string_01_r [{
   shapeType=Capsule
   length=2.6344
   radius=0.8756
   shapeName=dyn_string_01_r_capsule
   center=<1.9029 0.0344 0.0244>
   rotation=<rot 0 90.5123 -88.6249>
  }]
 dyn_string_02_l [{
   shapeType=Capsule
   length=3.9888
   radius=0.6567
   shapeName=dyn_string_02_l_capsule
   center=<2.3926 -0.0473 -0.0081>
   rotation=<rot 0 -92.2113 -87.3474>
  }]
 dyn_string_02_r [{
   shapeType=Capsule
   length=4.3488
   radius=0.8023
   shapeName=dyn_string_02_r_capsule
   center=<2.4554 0.0291 -0.1278>
   rotation=<rot 0 -89 -85.384>
  }]
 dyn_string_base_l [{shapeType=Sphere, radius=0.5, shapeName=dyn_string_base_l_sphere}]
 dyn_string_base_r [{shapeType=Sphere, radius=0.5, shapeName=dyn_string_base_r_sphere}]
 dyn_string_parent_l [{
   shapeType=Capsule
   length=2.2332
   radius=0.5268
   shapeName=dyn_string_parent_l_capsule
   center=<1.3356 0.0407 0.1475>
   rotation=<rot 0 92.4208 -81.4816>
  }]
 dyn_string_parent_r [{
   shapeType=Capsule
   length=2.3578
   radius=0.6522
   shapeName=dyn_string_parent_r_capsule
   center=<1.7541 0.0296 0.127>
   rotation=<rot 0 94.6091 -79.2201>
  }]
 spine_05 [
  {
   shapeType=Box
   extentX=27.2458
   extentY=7.5
   extentZ=25.9473
   shapeName=spine_05_box
   center=<2.5852 4.2227 -0.0002>
   rotation=<rot 0 -30.9106 0>
  }
  {
   shapeType=Box
   extentX=14.1666
   extentY=2.529
   extentZ=13.1769
   shapeName=spine_05_box
   center=<1.425 9.9855 1.0268>
   rotation=<rot 0 -34.1569 0>
  }
 ]
}
constraints [
 {bone1Name=dyn_string_parent_r, bone2Name=dyn_string_base_r, swing1LimitDegrees=10, swing2LimitDegrees=10, twistLimitDegrees=3}
 {bone1Name=dyn_string_01_r, bone2Name=dyn_string_parent_r, swing1LimitDegrees=15, swing2LimitDegrees=15, twistLimitDegrees=3}
 {bone1Name=dyn_string_02_r, bone2Name=dyn_string_01_r, swing1LimitDegrees=20, swing2LimitDegrees=20, twistLimitDegrees=3}
 {bone1Name=dyn_string_parent_l, bone2Name=dyn_string_base_l, swing1LimitDegrees=10, swing2LimitDegrees=10, twistLimitDegrees=3}
 {bone1Name=dyn_string_01_l, bone2Name=dyn_string_parent_l, swing1LimitDegrees=15, swing2LimitDegrees=15, twistLimitDegrees=3}
 {bone1Name=dyn_string_02_l, bone2Name=dyn_string_01_l, swing1LimitDegrees=20, swing2LimitDegrees=20, twistLimitDegrees=3}
 {bone1Name=dyn_hood_parent, bone2Name=dyn_hood_base, swing1LimitDegrees=20, swing2LimitDegrees=30, twistLimitDegrees=3}
]
