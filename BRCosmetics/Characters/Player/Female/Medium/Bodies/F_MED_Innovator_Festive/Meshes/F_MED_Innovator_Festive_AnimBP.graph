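# Secondary-motion AnimBP for the Innovator Festive female medium body cosmetic, layered over the pose copied from the attached parent mesh.
# Weighted constraints place a pouch base and both kneepads, with additive kneepad offsets while crouching.
# Four spring-driven AnimDynamics nodes move the pouches, then a RigidBody node simulates the physics asset relative to spine_05.
# The update event converts crouch and frontend booleans into alphas, reducing RigidBody influence in the frontend.
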AnimBlueprint /BRCosmetics/Characters/Player/Female/Medium/Bodies/F_MED_Innovator_Festive/Meshes/F_MED_Innovator_Festive_AnimBP
target_skeleton=&/F_MED_Innovator_Festive_Skeleton
graphs [
 AnimGraph {nodes [
   n1 = "Constraint - dyn_pouch_04_base (PPPPP)"!(
    ComponentPose=Local To Component(LocalPose=Copy Pose From Mesh() {bUseAttachedParent=true})
    Alpha=1
    ConstraintWeights_0=0.2663
    ConstraintWeights_1:float=0.3472
    ConstraintWeights_2:float=0.2428
    ConstraintWeights_3:float=0.1285
    ConstraintWeights_4:float=0.0152
   ) -> (Pose) {
    alphaBoolBlend {blendOption=Linear}
    boneToModify {boneName=dyn_pouch_04_base}
    constraintSetup [
     {targetBone {boneName=spine_02}}
     {targetBone {boneName=spine_04}}
     {targetBone {boneName=spine_03}}
     {targetBone {boneName=spine_01}}
     {targetBone {boneName=pelvis}}
    ]
    constraintWeights [1,1,1,1,1]
   }
   n2 = "Constraint - dyn_kneepad_parent_l (PP)"!(ComponentPose=@n1, Alpha=1, ConstraintWeights_0=0.5913, ConstraintWeights_1:float=0.4087) -> (Pose) {
    alphaBoolBlend {blendOption=Linear}
    boneToModify {boneName=dyn_kneepad_parent_l}
    constraintSetup [{targetBone {boneName=thigh_l}}, {targetBone {boneName=calf_l}}]
    constraintWeights [1,1]
   }
   n3 = "Constraint - dyn_kneepad_parent_r (PP)"!(ComponentPose=@n2, Alpha=1, ConstraintWeights_0=0.6198, ConstraintWeights_1:float=0.3802) -> (Pose) {
    alphaBoolBlend {blendOption=Linear}
    boneToModify {boneName=dyn_kneepad_parent_r}
    constraintSetup [{targetBone {boneName=thigh_r}}, {targetBone {boneName=calf_r}}]
    constraintWeights [1,1]
   }
   n4 = Get IsCrouching() -> (IsCrouching:double)
   n5 = "Transform (Modify) Bone - Bone: dyn_kneepad_parent_r"(ComponentPose=@n3, Alpha=@n4, Translation="-0.000,-0.000,-2.25", Rotation="30.584,0.000,0.000", Scale="1,1,1") {
    rotationMode=BMM_Additive
    rotationSpace=BCS_ParentBoneSpace
    translationMode=BMM_Additive
    translationSpace=BCS_ParentBoneSpace
    alphaBoolBlend {blendOption=Linear}
    alphaScaleBiasClamp {bInterpResult=true}
    boneToModify {boneName=dyn_kneepad_parent_r}
    rotation {pitch=30.584}
    translation {z=-2.25}
   }
   n6 = "Transform (Modify) Bone - Bone: dyn_kneepad_parent_l"(ComponentPose=@n5, Alpha=@n4, Translation="-0.000,-0.000,-2.25", Rotation="30.584,0.000,0.000", Scale="1,1,1") {
    rotationMode=BMM_Additive
    rotationSpace=BCS_ParentBoneSpace
    translationMode=BMM_Additive
    translationSpace=BCS_ParentBoneSpace
    alphaBoolBlend {blendOption=Linear}
    alphaScaleBiasClamp {bInterpResult=true}
    boneToModify {boneName=dyn_kneepad_parent_l}
    rotation {pitch=30.584}
    translation {z=-2.25}
   }
   n7 = Anim Dynamics - Bone: dyn_pouch_03_parent(ComponentPose=@n6, Alpha=1, ExternalForce="0,0,0") {
    angularDampingOverride=0.9999
    bAngularSpring=true
    bGravityOverrideInSimSpace=true
    bLinearSpring=true
    bOverrideAngularDamping=true
    bOverrideLinearDamping=true
    bUseGravityOverride=true
    linearDampingOverride=0.999
    simulationSpace=BoneRelative
    alphaBoolBlend {blendOption=Linear}
    boundBone {boneName=dyn_pouch_03_parent}
    componentAppliedLinearAccClamp=<8000 8000 8000>
    componentLinearAccScale=<0.15 0.15 0.15>
    componentLinearVelScale=<0.15 0.15 0.15>
    gravityOverride=<1 1 1>
    physicsBodyDefinitions [{
      sphereCollisionRadius=0
      boundBone {boneName=dyn_pouch_03_parent}
      boxExtents=<12 15 4>
      constraintSetup {angularLimitsMax=<0 5 5>, angularLimitsMin=<0 -2 -5>, angularTarget=<0 0 0>}
      localJointOffset=<4.5 0 0>
     }]
    relativeSpaceBone {boneName=pelvis}
   }
   n8 = Anim Dynamics - Bone: dyn_pouch_02_parent(ComponentPose=@n7, Alpha=0.5, ExternalForce="0,0,0") {
    angularDampingOverride=0.9999
    bAngularSpring=true
    bGravityOverrideInSimSpace=true
    bLinearSpring=true
    bOverrideAngularDamping=true
    bOverrideLinearDamping=true
    bUseGravityOverride=true
    linearDampingOverride=0.9999
    simulationSpace=BoneRelative
    alphaBoolBlend {blendOption=Linear}
    boundBone {boneName=dyn_pouch_02_parent}
    componentAppliedLinearAccClamp=<6000 6000 6000>
    componentLinearAccScale=<0.15 0.15 0.15>
    componentLinearVelScale=<0.15 0.15 0.15>
    gravityOverride=<1 1 1>
    physicsBodyDefinitions [{
      sphereCollisionRadius=0
      boundBone {boneName=dyn_pouch_02_parent}
      boxExtents=<15 20 4>
      constraintSetup {angularLimitsMax=<0 5 5>, angularLimitsMin=<0 -2 -5>, angularTarget=<0 0 0>}
      localJointOffset=<4.5 0 0>
     }]
    relativeSpaceBone {boneName=root}
   }
   n9 = Anim Dynamics - Bone: dyn_pouch_04_parent(ComponentPose=@n8, Alpha=0.75, ExternalForce="0,0,0") {
    angularDampingOverride=0.9999
    bAngularSpring=true
    bGravityOverrideInSimSpace=true
    bLinearSpring=true
    bOverrideAngularDamping=true
    bOverrideLinearDamping=true
    bUseGravityOverride=true
    linearDampingOverride=0.9999
    simulationSpace=BoneRelative
    alphaBoolBlend {blendOption=Linear}
    boundBone {boneName=dyn_pouch_04_parent}
    componentAppliedLinearAccClamp=<6000 6000 6000>
    componentLinearAccScale=<0.15 0.15 0.15>
    componentLinearVelScale=<0.15 0.15 0.15>
    gravityOverride=<1 1 1>
    physicsBodyDefinitions [{
      sphereCollisionRadius=0
      boundBone {boneName=dyn_pouch_04_parent}
      boxExtents=<12 15 4>
      constraintSetup {angularLimitsMax=<0 5 5>, angularLimitsMin=<0 -2 -5>, angularTarget=<0 0 0>}
      localJointOffset=<4.5 0 0>
     }]
    relativeSpaceBone {boneName=root}
   }
   n10 = Anim Dynamics - Bone: dyn_pouch_01_parent(ComponentPose=@n9, Alpha=1, ExternalForce="0,0,0") {
    angularDampingOverride=0.9999
    bAngularSpring=true
    bGravityOverrideInSimSpace=true
    bLinearSpring=true
    bOverrideAngularDamping=true
    bOverrideLinearDamping=true
    bUseGravityOverride=true
    linearDampingOverride=0.9999
    simulationSpace=BoneRelative
    alphaBoolBlend {blendOption=Linear}
    boundBone {boneName=dyn_pouch_01_parent}
    componentAppliedLinearAccClamp=<6000 6000 6000>
    componentLinearAccScale=<0.15 0.15 0.15>
    componentLinearVelScale=<0.15 0.15 0.15>
    gravityOverride=<1 1 1>
    physicsBodyDefinitions [{
      sphereCollisionRadius=0
      boundBone {boneName=dyn_pouch_01_parent}
      boxExtents=<12 20 4>
      constraintSetup {angularLimitsMax=<0 5 5>, angularLimitsMin=<0 -2 -5>, angularTarget=<0 0 0>}
      localJointOffset=<4.5 0 0>
     }]
    relativeSpaceBone {boneName=root}
   }
   n11 = RigidBody(ComponentPose=@n10, Alpha=Get IsInFrontend() -> (IsInFrontend:double)) {
    bClampLinearTranslationLimitToRefPose=true
    bOverrideWorldGravity=true
    overridePhysicsAsset=&/F_MED_Innovator_Festive_Physics
    simulationSpace=BaseBoneSpace
    alphaBoolBlend {blendOption=Linear}
    baseBoneRef {boneName=spine_05}
    componentAppliedLinearAccClamp=<6000 6000 6000>
    componentLinearAccScale=<0.15 0.15 0.15>
    componentLinearVelScale=<0.15 0.15 0.15>
    overrideWorldGravity {z=-980}
   }
   Output Pose(Result=Component To Local(ComponentPose=@n11)) {layerGroup=None}
  ]}
 EventGraph {nodes [
   Event BlueprintUpdateAnimation()
   > n1 = Sequence()
   > Set IsCrouching(IsCrouching:double=SelectFloat(A=1, B=0, bPickA=Get bIsCrouching())) -> (Output_Get:double)
   Set IsInFrontend(execute=@n1.then_1, IsInFrontend:double=SelectFloat(A=0.43, B=1, bPickA=Get bIsInFrontEnd())) -> (Output_Get:double)
  ]}
]
variables [
 IsCrouching {type=number}
 IsInFrontend {type=number}
]
